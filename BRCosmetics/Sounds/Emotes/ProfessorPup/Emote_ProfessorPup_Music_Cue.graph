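# Sound cue for the Professor Pup emote music, set up as first-person (1P) emote music through its sound class, property sheet and concurrency rule.
# A mixer starts a one-shot intro alongside a looping wave, which a fixed delay of about 2.7 seconds holds back to follow the intro.
# An enveloper over the delayed loop has a near-instant volume ramp, and the cue has a send to a music audio-analysis submix.
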SoundCue /BRCosmetics/Sounds/Emotes/ProfessorPup/Emote_ProfessorPup_Music_Cue
root=Mixer_0
nodes [
 {id=Delay_0, children [WavePlayer_3], properties {delayMax=2.666, delayMin=2.666}}
 {
  id=Enveloper_0
  children [Delay_0]
  properties {pitchCurve {editorCurveData {keys [{time=0}]}}, volumeCurve {editorCurveData {keys [<at 0 0>,<at 0.0004 1>]}}}
 }
 {id=Mixer_0, children [WavePlayer_1,Enveloper_0], properties {inputVolume [1,1]}}
 {id=WavePlayer_1, properties {soundWaveAssetPtr=&/Emote_Professsor_Pup_Intro}}
 {id=WavePlayer_3, properties {bLooping=true, soundWaveAssetPtr=&/Emote_Professsor_Pup_Loop}}
]
properties {
 audioPropertiesSheet=/BRCosmetics/Sounds/Settings/PropertySheets/Emotes/APS_Cosmetic_Emote_Music_1P
 maxDistance=2097152
 soundClassObject=/Game/Sounds/FortSoundClassesAndModes/Fort_Music_Emote_1P_PSM
 volumeMultiplier=1
 concurrencySet [/Game/Sounds/Concurrency_Rules/Athena_EmoteMusic_1P_Conc]
 soundSubmixSends [{minSendLevel=0, soundSubmix=&Fortnite_AudioAnalysis_Music}]
}
