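# Third-person music cue for the Professor Pup emote in Battle Royale, playing an intro followed by a looping track.
# A Mixer starts both branches together, with the intro slightly boosted and the loop held back by a fixed Delay to follow it.
# An Enveloper over the delayed branch applies a near-instant volume ramp.
# Emote attenuation, concurrency and sound class apply; playback continues when silent and feeds a music-analysis submix.
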SoundCue /BRCosmetics/Sounds/Emotes/ProfessorPup/SC_Emote_ProfessorPup_Music_Cue_3p
root=Mixer_1
nodes [
 {id=Delay_1, children [WavePlayer_3], properties {delayMax=2.677, delayMin=2.677}}
 {
  id=Enveloper_0
  children [Delay_1]
  properties {pitchCurve {editorCurveData {keys [{time=0}]}}, volumeCurve {editorCurveData {keys [<at 0 0>,<at 0.0004 1>]}}}
 }
 {id=Mixer_1, children [WavePlayer_1,Enveloper_0], properties {inputVolume [1.25,1]}}
 {id=WavePlayer_1, properties {soundWaveAssetPtr=&/Emote_Professsor_Pup_Intro}}
 {id=WavePlayer_3, properties {bLooping=true, soundWaveAssetPtr=&/Emote_Professsor_Pup_Loop}}
]
properties {
 attenuationSettings=&Fort_Emotes_Athena_Att
 audioPropertiesSheet=/BRCosmetics/Sounds/Settings/PropertySheets/Emotes/APS_Cosmetic_Emote_Music_3P
 maxDistance=1600
 soundClassObject=/Game/Sounds/FortSoundClassesAndModes/Fort_Music_Emote_3P
 virtualizationMode=PlayWhenSilent
 volumeMultiplier=1
 concurrencySet [/Game/Sounds/Concurrency_Rules/Athena_EmoteMusic_Conc]
 soundSubmixSends [{minSendLevel=0, soundSubmix=&Fortnite_AudioAnalysis_Music}]
}
